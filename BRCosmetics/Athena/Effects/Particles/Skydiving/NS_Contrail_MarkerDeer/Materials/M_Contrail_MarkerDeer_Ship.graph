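# Default Lit Niagara mesh-particle material for the ship in the MarkerDeer skydiving contrail cosmetic.
# Surface attributes come from 2023 character-parent texture and BasicTechArt functions, with a dynamic parameter offsetting V and a Fresnel input.
# A second dynamic parameter's Aperture, or its inverse, sets RotateAboutAxis angles around the mesh particle pivot across four local Y/Z position-masked regions.
# Several nodes appear unconnected, including a debug sine remap and emissive texture sample.
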Material /BRCosmetics/Athena/Effects/Particles/Skydiving/NS_Contrail_MarkerDeer/Materials/M_Contrail_MarkerDeer_Ship
graph {
 properties {bAutomaticallySetUsageInEditor=false, bUsedWithNiagaraMeshParticles=true, pixelDepthOffsetMode=PDOM_Legacy, usedShadingModels=Default Lit}
 nodes [
  Add()
  Constant3Vector() {constant=<rgba 1 0 0 0>}
  Constant()
  n1 = Constant() {r=1}
  n2 = LocalPosition()
  n3 = VertexInterpolator(VS=@n2.XYZ)
  n4 = MaterialFunctionCall(Float2=TextureCoordinate()) {materialFunction=&BreakOutFloat2Components}
  n5 = DynamicParameter()
  n6 = MaterialFunctionCall!(UV=AppendVector(A=@n4.R, B=Add(A=@n5.Param1, B=@n4.G))) {materialFunction=/Game/Global/Materials/Character_Parent_2023/Functions/MF_CharacterParent_2023_TextureInputs}
  GetMaterialAttributes(None=MaterialFunctionCall!(
    Tech Art Mask UVs
    Emissive UVs
    Fresnel=Fresnel(ExponentIn=@n1) {baseReflectFraction=0}
    LocalPosition=ShaderStageSwitch(PixelShader=@n3.PS, VertexShader=@n2.XYZ)
    Mask Input=@n6.M
    Material=@n6.M
   ) {materialFunction=/Game/Global/Materials/Character_Parent_2023/Functions/MF_CharacterParent_2023_BasicTechArt}.Result) {attributeGetTypes [
    #1
    #2
    #3
    #4
    #5
   ]}
  n7 = VectorParameter() {defaultValue=<rgba -1 1 -0.2 -0.01>}
  MaterialFunctionCall(
   Input=MaterialFunctionCall() {materialFunction=&DebugTimeSine}.Result
   Input Low=@n7.R
   Input High=@n7.R
   Target Low=@n7.R
   Target High=@n7.R
  ) {materialFunction=&RemapValueRange}
  ParticleDirection()
  PreSkinnedPosition()
  n8 = DynamicParameter() {parameterIndex=1, paramNames [XAxis X, XAxis Y, XAxis Z, Aperture]}
  n9 = LocalPosition()
  n10 = Add(A=@n9.XYZ) {constB=-2}
  n11 = Ceil(None=Saturate(None=ComponentMask(None=@n10) {g=true}))
  n12 = Ceil(None=Saturate(None=ComponentMask(None=@n9.XYZ) {b=true}))
  n13 = Multiply(A=@n8.Aperture, B=Multiply(A=@n11, B=@n12))
  n14 = Add(A=OneMinus(None=@n9.XYZ)) {constB=-3}
  n15 = Ceil(None=Saturate(None=ComponentMask(None=@n14) {g=true}))
  n16 = OneMinus(None=@n8.Aperture)
  n17 = Multiply(A=Multiply(A=@n12, B=@n15), B=@n16)
  n18 = OneMinus(None=@n12)
  n19 = Multiply(A=@n8.Aperture, B=Multiply(A=@n15, B=@n18))
  n20 = Multiply(A=@n16, B=Multiply(A=@n11, B=@n18))
  RotateAboutAxis(
   NormalizedRotationAxis=MaterialFunctionCall(R=@n8.XAxis X, G=@n8.XAxis X, B=@n8.XAxis X) {materialFunction=&AppendMany}.RGB
   RotationAngle=Add(A=Add(A=@n13, B=@n17), B=Add(A=@n19, B=@n20))
   PivotPoint=MaterialFunctionCall() {materialFunction=&ObjectPivotPoint}.Mesh Particle Pivot Location
   Position=WorldPosition().XYZ
  )
  ScalarParameter() {parameterName=Aperture, sliderMin=-0.03}
  TextureSample() {texture=/BRCosmetics/Athena/Effects/Particles/Skydiving/NS_Contrail_MarkerDeer/Textures/T_MarkerDeer_Contrail_E}
 ]
}
