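# MarkerDeer skydiving contrail Niagara system that flies a five-member mesh formation, from emitters named Ship, around the skydiver's skeletal mesh.
# Ship (gameplay) and Ship_FE (front end) spawn in staggered bursts, then follow sampled mesh positions with per-index offsets, waveform wobble and smoothed orientation.
# Trail emitters read those particles to place glow sprites, while GPU speed lines stream near the player.
# Dive value and system timelines drive formation, scale, brightness and spawn rates.
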NiagaraSystem /BRCosmetics/Athena/Effects/Particles/Skydiving/NS_Contrail_MarkerDeer/NS_Contrail_MarkerDeer
dependencies {
 usedDataInterfaces [
  &NiagaraDataInterfaceCurve
  &NiagaraDataInterfaceArrayFloat3
  &NiagaraDataInterfaceArrayFloat
  &NiagaraDataInterfaceVector2DCurve
  &NiagaraDataInterfaceParticleRead
  &NiagaraDataInterfaceSkeletalMesh
 ]
 usedDynamicInputs [&Vector2DFromCurve, &Divide_Float]
 usedModules [
  &self:SystemSpawnScript.SystemScriptSource.SystemScriptGraph.NiagaraNodeAssignment_0.SetVariables_#1
  &self:SystemSpawnScript.SystemScriptSource.SystemScriptGraph.NiagaraNodeAssignment_2.SetVariables_#2
  &SystemState
  /Game/Effects/Niagara/Modules/Skydiving/FN_DivingAttributes/v6/FN_DivingAttributes
  /Game/Effects/Niagara/Modules/Contrails/FN_Contrail_Setup
  &self:ScratchModule
  &self:SystemSpawnScript.SystemScriptSource.SystemScriptGraph.NiagaraNodeAssignment_4.SetVariables_#3
  &self:SystemSpawnScript.SystemScriptSource.SystemScriptGraph.NiagaraNodeAssignment_1.SetVariables_#4
  &FN_Integrator
  &Timeline
  &self:SystemSpawnScript.SystemScriptSource.SystemScriptGraph.NiagaraNodeAssignment_3.SetVariables_#5
 ]
}
emitters {
 FN_SpeedLines {
  simTarget=GPUComputeSim
  emitterUpdateScript [
   &EmitterState(Life Cycle Mode=::NewEnumerator1, Loop Duration=1, UseLoopDelay=false)
   &SpawnRate(
    Spawn Probability=&MakeCustomFloatFromBool(Bool=&SetBoolByFloatComparison(A=$System.Age, B=0.2), True Float=1, False Float=1)
    SpawnRate=&Lerp_Float(
     A=0
     B=&FN_FrontEndFloatModifier(Gameplay=&Multiply_Float(A=55, B=$System.FN_Contrail_Setup.Dive Value), Front End=0)
     Alpha=1
    )
    Use Spawn Probability=true
   )
   &SpawnRate(
    Spawn Probability=&MakeCustomFloatFromBool(Bool=&SetBoolByFloatComparison(A=$System.Age, B=0.2), True Float=1, False Float=1)
    SpawnRate=&Lerp_Float(A=0, B=500, Alpha=$System.Dive)
    Use Spawn Probability=true
   ) {enabled=false, module=SpawnRate001}
  ]
  particleSpawnScript [
   &InitializeParticle(
    Color Maximum=<rgba 0.25 0.25 0.25 0.05>
    Color Minimum=<rgba 1 1 1 0.025>
    Color Mode=::NewEnumerator2
    Lifetime Max=0.5
    Lifetime Min=0.3
    Lifetime Mode=::NewEnumerator1
    Sprite Rotation Mode=::NewEnumerator2
    Sprite Size=&UniformRangedVector2D(Minimum=<7 100>, Maximum=<1 200>)
    Sprite Size Mode=::NewEnumerator4
    UsePositionOffset=false
    Write Material Random=true
   ) {module=InitializeParticle001}
   &SkeletalMeshLocation(
    Bone / Socket Index=3
    Bone Sampling Mode=::NewEnumerator4
    MaskSpawnGroup=false
    Sampled Orientation Coordinate Space=::Local
    Skeletal Mesh=$System.SkeletalMesh
   )
   &SphereLocation(
    MaskSpawnGroup=false
    Non Uniform Scale=&RotateVectorByQuaternion(Vector To Rotate=<1 1 1>, Quaternion=$Particles.SkeletalMeshLocation.SampledRotation)
    Offset=&FN_FrontEndVector3Modifier(
     Gameplay=&Add_Vector(
      A=&RotateVectorByQuaternion(Vector To Rotate=<100 0 0>, Quaternion=$Particles.SkeletalMeshLocation.SampledRotation)
      B=&TransformVector(Vector=<0 0 -200>, Source Space=::World, Destination Space=::World)
     )
     Front End=<0 0 0>
    )
    Override Seed=false
    Sphere Orientation Axis:Vector3f=&RotateVectorByQuaternion(Vector To Rotate=<100 0 0>, Quaternion=$Particles.SkeletalMeshLocation.SampledRotation)
    Sphere Origin:NiagaraPosition=$Particles.SkeletalMeshLocation.SampledPosition
    Sphere Radius=250
    Surface Only=false
    UseNonUniformScale=true
    UseOffset=true
    UseSphereOrientation=true
   )
   &InheritVelocity(Module.Limit Speed=true)
   &AddVelocity(Velocity=&FN_FrontEndVector3Modifier(Gameplay=<0 0 0>, Front End=<0 -100 100>))
  ]
  particleUpdateScript [
   &ParticleState()
   &SolveForcesAndVelocity(Clamp Velocity=false, Limit Acceleration=false)
   &ScaleSpriteSize(
    Non-Uniform Curve Scale=<1.5 3>
    Non-Uniform Curve Sprite Scale=DataInterface() {propertyValues {
      XCurve {keys [<at -0.0069 0.5 RCIM_Cubic>,<at 0.4918 1 RCIM_Cubic>,<at 1 0.5 RCIM_Cubic>]}
      YCurve {keys [
        {
         arriveTangent=3.1787
         arriveTangentWeight=0.684
         interpMode=RCIM_Cubic
         leaveTangent=3.1787
         leaveTangentWeight=0.684
         tangentMode=RCTM_User
         tangentWeightMode=RCTWM_WeightedBoth
         time=0
         value=0.25
        }
        {
         arriveTangentWeight=0.2354
         interpMode=RCIM_Cubic
         leaveTangentWeight=0.3527
         tangentMode=RCTM_Break
         tangentWeightMode=RCTWM_WeightedBoth
         time=0.4975
         value=1
        }
        <at 1 0.25 RCIM_Cubic>
       ]}
     }}
    Scale Sprite Size Mode=::NewEnumerator3
   )
  ]
  rendererClasses [&NiagaraSpriteRendererProperties]
  renderers [{rendererClass=&NiagaraSpriteRendererProperties}]
 }
 Ship {
  emitterUpdateScript [
   &EmitterState(Life Cycle Mode=::NewEnumerator1, UseLoopDelay=false)
   &SpawnBurst_Instantaneous(
    Age:NiagaraFloat=$Emitter.Age
    Spawn Count=&MakeIntFromFloat(Float=&FN_FrontEndFloatModifier(Gameplay=1, Front End=0))
    Use Spawn Probability=false
    UseLoopCountLimit=false
   )
   &SpawnBurst_Instantaneous(
    Age:NiagaraFloat=$Emitter.Age
    Spawn Count=&MakeIntFromFloat(Float=&FN_FrontEndFloatModifier(Gameplay=2, Front End=0))
    Spawn Time=0.5
    Use Spawn Probability=false
    UseLoopCountLimit=false
   ) {module=SpawnBurst_Instantaneous001}
   &SpawnBurst_Instantaneous(
    Age:NiagaraFloat=$Emitter.Age
    Spawn Count=&MakeIntFromFloat(Float=&FN_FrontEndFloatModifier(Gameplay=2, Front End=0))
    Spawn Time=1
    Use Spawn Probability=false
    UseLoopCountLimit=false
   ) {module=SpawnBurst_Instantaneous002}
   &InterpolateOverTime(Rate Of Change=50, Target Vector Value:Vector3f=$Engine.Owner.Velocity, Variable Type=::NewEnumerator2)
   /SetVariables(Emitter.ExhaustBlend=0.5) {isSetParameters=true, module=SetVariables_#6}
   /SetVariables(Emitter.WindWeight:NiagaraFloat=$System.FN_Contrail_Setup.Dive Value) {isSetParameters=true, module=SetVariables_#7}
  ]
  particleSpawnScript [
   &InitializeParticle(
    Color Mode=::NewEnumerator0
    Lifetime=5
    Mesh Scale Mode=::NewEnumerator3
    Mesh Uniform Scale=&Multiply_Float(A=$System.Ship Scale, B=0.7)
    Position Offset=<0 0 1000>
    UsePositionOffset=true
    Write Material Random=true
   )
   &ShapeLocation(MaskSpawnGroup=false, Override Seed=false, UseOwnerScale=true)
   &InitialMeshOrientation(Mesh Orientation Mode=::NewEnumerator7, Use Rotation Vector=true)
   &FN_SetSpawnGroup(Number of Spawn Groups=5)
   /SetVariables(Particles.ShipNum:NiagaraInt32=$Emitter.SpawnGroup) {isSetParameters=true, module=SetVariables_#8}
  ]
  particleUpdateScript [
   &ParticleState(Kill Particles When Lifetime Has Elapsed=false, Let Infinitely Lived Particles Die When Emitter Deactivates=true)
   /SetVariables(Particles.Ambient:Vector3f=&Multiply_VectorByFloat(
     Vector=&MakeVector(
      X=&Waveform(
       "Waveform [1]"=::NewEnumerator12
       "[1] Input Position"=&RandomRangeFloat(Minimum=0, Maximum=1000)
       "[1] Frequency"=1
       "[1] Amplitude Min/Max"=<-50 50>
      )
      Y=&Waveform(
       "Waveform [1]"=::NewEnumerator12
       "[1] Input Position"=&RandomRangeFloat(Minimum=1200, Maximum=1800)
       "[1] Frequency"=1
       "[1] Amplitude Min/Max"=<-50 50>
      )
      Z=&Waveform(
       "Waveform [1]"=::NewEnumerator12
       "[1] Input Position"=&RandomRangeFloat(Minimum=2200, Maximum=2800)
       "[1] Frequency"=1
       "[1] Amplitude Min/Max"=<-50 50>
      )
     )
     Float=&Lerp_Float(A=0.02, B=0.1, Alpha=$System.FN_Contrail_Setup.Dive Value)
    )) {isSetParameters=true, module=SetVariables_#9}
   &InterpolateOverTime(
    Rate Of Change=50
    Target Float Value:NiagaraFloat=&MakeCustomFloatFromBool(Bool=$System.Timeline001.IsRewinding, True Float=0, False Float=1)
   ) {module=InterpolateOverTime004}
   /SetVariables(Particles.CatchupVector:Vector3f=&Multiply_VectorByFloat(
     Vector=&Multiply_VectorByFloat(
      Vector=&NormalizeVector(Vector To Normalize=$System.SkydiveVelocity, Safe Normalize=true, Safe Fallback Threshold=0, Safe Fallback Vector=<0 0 0.001>)
      Float=-75
     )
     Float=&Lerp_Float(
      A=&FloatFromCurve(
       FloatCurve=DataInterface() {propertyValues {Curve {keys [
           {interpMode=RCIM_Cubic, tangentMode=RCTM_User, time=0, value=0}
           {interpMode=RCIM_Cubic, tangentMode=RCTM_User, time=0.35, value=-0.35}
           {tangentMode=RCTM_User, time=1, value=0}
          ]}}}
       CurveIndex=$System.Timeline001.NormalizedTime
      )
      B=&FloatFromCurve(
       FloatCurve=DataInterface() {propertyValues {Curve {keys [
           {arriveTangent=3.3827, arriveTangentWeight=1.1106, interpMode=RCIM_Cubic, leaveTangentWeight=1.1106, time=0, value=0}
           {arriveTangentWeight=0.5241, interpMode=RCIM_Cubic, leaveTangentWeight=0.1667, time=0.5, value=0.65}
           <at 1 0 RCIM_Cubic>
          ]}}}
       CurveIndex=$System.Timeline001.NormalizedTime
      )
      Alpha=$Particles.InterpolateOverTime004.InterpolateOverTime_Float.Moving Average
     )
    )) {isSetParameters=true, module=SetVariables_#10}
   &SkeletalMeshLocation(
    Bone Sampling Mode=::NewEnumerator4
    MaskSpawnGroup=false
    Sampled Position Offset=&Add_Vector(
     A=$Particles.CatchupVector
     B=/Niagara/DynamicInputs/Vectors/RotateVector(
      Vector To Rotate:Vector3f=&Add_Vector(
       A=$Particles.Ambient
       B=&Add_Vector(
        A=&Lerp_Vector(
         A=&SelectVectorFromArray(
          Direct Array Index=&ReturnExecIndex()
          Direct Array Mode=::NewEnumerator1
          Vector Selection Array=DataInterface() {propertyValues {FloatData [<0 60 120>, <-25 35 100>, <25 35 100>, <-50 0 80>, <50 0 80>]}}
         )
         B=&SelectVectorFromArray(
          Direct Array Index=&ReturnExecIndex()
          Direct Array Mode=::NewEnumerator1
          Vector Selection Array=DataInterface() {propertyValues {FloatData [<0 70 140>, <-25 35 140>, <25 35 140>, <-50 0 140>, <50 0 140>]}}
         )
         Alpha=&FloatFromCurve(
          FloatCurve=DataInterface() {propertyValues {Curve {
             defaultValue=0
             keys [
              {arriveTangent=0.2783, interpMode=RCIM_Cubic, leaveTangent=0.2783, tangentMode=RCTM_User, time=-0.2, value=0}
              {arriveTangent=2.7714, tangentMode=RCTM_User, time=0.8, value=1}
             ]
            }}}
          CurveIndex=$System.FN_Contrail_Setup.Dive Value
         )
        )
        B=&Lerp_Vector(
         A=<0 0 500>
         B=<0 0 75>
         Alpha=&FloatFromCurve(
          FloatCurve=DataInterface() {propertyValues {Curve {keys [{arriveTangent=1.2212, interpMode=RCIM_Cubic, leaveTangent=1.2212, tangentMode=RCTM_User, time=0, value=0}, <at 3 1 RCIM_Cubic>]}}}
          CurveIndex=$Particles.Age
         )
        )
       )
      )
      Yaw:NiagaraFloat=&Waveform(
       "Waveform [1]"=::NewEnumerator12
       "[1] Input Position"=&RandomRangeFloat(Minimum=200000, Maximum=100000)
       "[1] Frequency"=1.25
       "[1] Amplitude Min/Max"=<7 -7>
       "[1] Scale"=0.2
      )
      Pitch:NiagaraFloat=0
      Roll:NiagaraFloat=0
      Rotation Rate:NiagaraFloat=1
     )
    )
    Skeletal Mesh=$System.SkeletalMesh
    Use Sampled Mesh Attribute as Ribbon ID=false
   )
   &ScaleMeshSize(Scale Factor=&VectorFromFloat(Value=&FloatFromCurve(
      FloatCurve=DataInterface() {propertyValues {Curve {keys [{arriveTangent=1.0186, interpMode=RCIM_Cubic, leaveTangent=1.0186, tangentMode=RCTM_User, time=0, value=0}, <at 0.3 1 RCIM_Cubic>]}}}
      CurveIndex=$Particles.Age
     )))
   &SolveForcesAndVelocity(Clamp Velocity=false, Limit Acceleration=false)
   &UpdateMeshOrientation(
    Bank Rate:NiagaraFloat=&Lerp_Float(A=1, B=900, Alpha=$System.FN_Contrail_Setup.Dive Value)
    Banking Angle Min / Max:Vector2f=&Lerp_Vector2(A=<0 0>, B=<0 0>, Alpha=$System.FN_Contrail_Setup.Dive Value)
    "Foward Vector (Mesh Space)":Vector3f=&Lerp_Vector(A=<1 0 1>, B=<1 0 0>, Alpha=$System.FN_Contrail_Setup.Dive Value)
    Initial Forward Vector:Vector3f=<0 -1 0>
    Orientation Method=::NewEnumerator3
    Particle Velocity:Vector3f=$Emitter.InterpolateOverTime.InterpolateOverTime_V3_.Moving Average
    "Right Vector (Mesh Space)":Vector3f=&Lerp_Vector(
     A=<0 1 -0.015>
     B=<0 1 0.015>
     Alpha=&MakeFloatFromVector(Vector=$Particles.Ambient, Channel=::NewEnumerator0)
    )
    Rotation Decay Rate In Degrees:NiagaraFloat=&Lerp_Float(A=150, B=500, Alpha=$System.Dive)
    Up Vector:Vector3f=&Lerp_Vector(
     A=&ReturnMeshOrientationAxis(Mesh Orientation Axis=::YAxis, Mesh Orientation Quat=$Particles.SkeletalMeshLocation.SampledRotation)
     B=&ReturnMeshOrientationAxis(Mesh Orientation Axis=::ZAxis, Mesh Orientation Quat=$Particles.SkeletalMeshLocation.SampledRotation)
     Alpha=&FloatFromCurve(FloatCurve=DataInterface() {propertyValues {Curve {keys [<at 0 -0.25>,<at 0.25 0.5>]}}}, CurveIndex=$System.FN_Contrail_Setup.Dive Value)
    )
    Upright Roll Correction Speed:NiagaraFloat=&Lerp_Float(A=25, B=90000, Alpha=$System.Dive)
    Velocity Orientation Convergence Rate:NiagaraFloat=&Lerp_Float(
     A=&Lerp_Float(
      A=3
      B=12
      Alpha=&Waveform(
       "Waveform [1]"=::NewEnumerator12
       "[1] Input Position"=&RandomRangeFloat(Minimum=0, Maximum=100)
       "[1] Frequency"=5
       "[1] Amplitude Min/Max"=<0 1>
      )
     )
     B=25
     Alpha=$System.Dive
    )
   ) {module=UpdateMeshOrientation001}
   /SetVariables(Particles.BankOrientation:Quat4f=&MultiplyQuaternion(
     Quaternion A=$Particles.MeshOrientation
     Quaternion B=&MakeQuaternion(
      Quaternion From=0
      Axis=<1 0 0>
      Angle=$System.BankAngle
      From=?
      To=?
      Angle Type=::NewEnumerator0
      Coordinate Space=::World
      Normalize Vectors=false
     )
    )) {isSetParameters=true}
   &InterpolateOverTime(Rate Of Change=5, Target Quat Value:Quat4f=$Particles.BankOrientation, Variable Type=::NewEnumerator5) {module=InterpolateOverTime002}
   /SetVariables(Particles.New Orient:Quat4f=$Particles.InterpolateOverTime002.InterpolateOverTime_Quaternion.Moving Average) {isSetParameters=true}
   /SetVariables(Particles.MeshOrientation=$Particles.New Orient) {isSetParameters=true}
   /SetVariables(Particles.SpeedConeColor:LinearColor=&MakeLinearColorFromVectorAndFloat(
     "Vector (RGB)"=<1 1 1>
     "Float (Alpha)"=&Multiply_Float(
      A=$System.FN_Contrail_Setup.Dive Value
      B=&FloatFromCurve(
       FloatCurve=DataInterface() {propertyValues {Curve {
          defaultValue=1
          keys [
           {interpMode=RCIM_Cubic, tangentMode=RCTM_User, time=0.5, value=0.8}
           {interpMode=RCIM_Cubic, tangentMode=RCTM_User, time=0.6, value=3}
           {tangentMode=RCTM_User, time=1, value=0.8}
          ]
         }}}
       CurveIndex=$System.Timeline002.Time
       Scale Curve=&Multiply_Float(
        A=0.7
        B=&Waveform(
         Global Input Position=&RandomRangeFloat(Minimum=0, Maximum=1)
         "Waveform [1]"=::NewEnumerator12
         "[1] Frequency"=&Lerp_Float(A=8, B=12, Alpha=$System.FN_Contrail_Setup.Dive Value)
         "[1] Amplitude Min/Max"=<0.75 1.2>
        )
       )
      )
     )
    )) {isSetParameters=true, module=SetVariables_#11}
   /SetVariables(Particles.ExhaustPosition:NiagaraPosition=/Niagara/DynamicInputs/Math/TransformVectorByMatrix(
     Vector:Vector3f=<-5 0 0>
     Transformation Matrix:NiagaraMatrix=/Niagara/DynamicInputs/Math/MakeTransform(
      Inverse:NiagaraBool=false
      Position:Vector3f=Linked() {linkedVariable {name=Particles.Position, type=NiagaraPosition}}
      Local Offset:Vector3f=<0 0 0>
      Scale:Vector3f=$Particles.Scale
      Rotation Interface:ENiagara_RotationDefinitionApproach=::NewEnumerator1
      Axis:Vector3f=?
      Quaternion:Quat4f=$Particles.MeshOrientation
      Rotation Matrix:NiagaraMatrix=?
      X Vector:Vector3f=?
      Yaw:NiagaraFloat=?
      Angle:NiagaraFloat=?
      Pitch:NiagaraFloat=?
      Y Vector:Vector3f=?
      Roll:NiagaraFloat=?
      Z Vector:Vector3f=?
     )
     Transform Vector as Position:NiagaraBool=true
    )) {isSetParameters=true, module=SetVariables_#12}
   &DynamicMaterialParameters(
    Index 0 Param 1=&SelectFloatFromArray(
     Direct Array Index=&ReturnExecIndex()
     Float Selection Array=DataInterface() {propertyValues {FloatData [0,0.5]}}
    )
    Index 0 Param 2=&RandomRangeFloat(Minimum=0, Maximum=1)
    Index 1 Param 1=&MakeFloatFromVector(
     Vector=&ReturnMeshOrientationAxis(Mesh Orientation Axis=::XAxis, Mesh Orientation Quat=$Particles.MeshOrientation)
     Channel=::NewEnumerator0
    )
    Index 1 Param 2=&MakeFloatFromVector(
     Vector=&ReturnMeshOrientationAxis(Mesh Orientation Axis=::XAxis, Mesh Orientation Quat=$Particles.MeshOrientation)
     Channel=::NewEnumerator1
    )
    Index 1 Param 3=&MakeFloatFromVector(
     Vector=&ReturnMeshOrientationAxis(Mesh Orientation Axis=::XAxis, Mesh Orientation Quat=$Particles.MeshOrientation)
     Channel=::NewEnumerator2
    )
    Index 1 Param 4=&Lerp_Float(
     A=-0.03
     B=0.01
     Alpha=&FloatFromCurve(
      FloatCurve=DataInterface() {propertyValues {Curve {keys [{arriveTangent=-0.0511, interpMode=RCIM_Cubic, leaveTangent=-0.0511, tangentMode=RCTM_User, time=0, value=0}, <at 1 1>]}}}
      CurveIndex=$System.Timeline003.Time
     )
    )
    Module.Param0WriteEnabled=true
    Module.Param0WriteEnabled1=true
    Module.Param1WriteEnabled=true
    Module.Param1WriteEnabled1=true
    Module.Param2WriteEnabled=true
    Module.Param2WriteEnabled1=true
    Module.Param3WriteEnabled=true
    Module.Param3WriteEnabled1=true
    Write Parameter Index 1=true
   )
   &FN_Mesh_Attributes(
    Color=&MultiplyLinearColor(
     Linear Color=<rgba 1 1 1 1>
     Scale Factor=&MakeVector4FromVectorAndFloat(
      "Vector (RGB)"=&VectorFromFloat(Value=&Waveform(
        Global Input Position=&RandomRangeFloat(Minimum=0, Maximum=1)
        "Waveform [1]"=::NewEnumerator12
        "[1] Frequency"=3
        "[1] Amplitude Min/Max"=<0.9 5>
       ))
      "Float (Alpha)"=1
     )
    )
    Create Color Attribute=true
    Create Scale Attribute=true
    Scale=&Multiply_Vector(
     A=&MakeVector(X=&Lerp_Float(A=0, B=1, Alpha=$Emitter.ExhaustBlend), Y=1, Z=1)
     B=&Lerp_Vector(
      A=&Lerp_Vector(
       A=<0.1 1.1 0.6>
       B=<0.1 1 1>
       Alpha=&FloatFromCurve(
        FloatCurve=DataInterface() {propertyValues {Curve {keys [{arriveTangent=-0.0511, interpMode=RCIM_Cubic, leaveTangent=-0.0511, tangentMode=RCTM_User, time=0, value=0}, <at 1 1>]}}}
        CurveIndex=$System.Timeline003.Time
       )
      )
      B=&Multiply_Vector(
       A=&Lerp_Vector(
        A=<0.1 1.1 0.6>
        B=<0.1 1 1>
        Alpha=&FloatFromCurve(
         FloatCurve=DataInterface() {propertyValues {Curve {keys [{arriveTangent=-0.0511, interpMode=RCIM_Cubic, leaveTangent=-0.0511, tangentMode=RCTM_User, time=0, value=0}, <at 1 1>]}}}
         CurveIndex=$System.Timeline003.Time
        )
       )
       B=<3 1 1>
      )
      Alpha=$System.FN_Contrail_Setup.Dive Value
     )
    )
    Write to Module Namespace=true
   ) {module=FN_Mesh_Attributes001}
   &FN_Mesh_Attributes(
    Color=&MultiplyLinearColor(
     Linear Color=<rgba 1 1 1 1>
     Scale Factor=&MakeVector4FromVectorAndFloat(
      "Vector (RGB)"=<1 1 1>
      "Float (Alpha)"=&FloatFromCurve(
       FloatCurve=DataInterface() {propertyValues {Curve {
          defaultValue=1
          keys [
           {arriveTangent=-0.2012, interpMode=RCIM_Cubic, leaveTangent=-0.2012, tangentMode=RCTM_User, time=0.5, value=0}
           {arriveTangent=4.0969, interpMode=RCIM_Cubic, leaveTangent=4.0969, tangentMode=RCTM_User, time=1, value=1}
          ]
         }}}
       CurveIndex=$Emitter.WindWeight
       Scale Curve=0.02
      )
     )
    )
    Create Color Attribute=true
    Create Scale Attribute=true
    Scale=&Multiply_Vector(A=$Particles.Scale, B=&Lerp_Vector(A=<1 1 1>, B=<1 1 1>, Alpha=$Emitter.WindWeight))
    Write to Module Namespace=true
   ) {module=FN_Mesh_Attributes002}
  ]
  rendererClasses [&NiagaraMeshRendererProperties]
  renderers [
   {rendererClass=&NiagaraMeshRendererProperties, rendererIndex=0}
   {rendererClass=&NiagaraMeshRendererProperties, rendererIndex=1}
   {rendererClass=&NiagaraMeshRendererProperties, rendererIndex=2}
   {rendererClass=&NiagaraMeshRendererProperties, rendererIndex=3}
  ]
 }
 Ship_FE {
  emitterUpdateScript [
   &EmitterState(Life Cycle Mode=::NewEnumerator1, UseLoopDelay=false)
   &SpawnBurst_Instantaneous(
    Age:NiagaraFloat=$Emitter.Age
    Spawn Count=&MakeIntFromFloat(Float=&FN_FrontEndFloatModifier(Gameplay=0, Front End=1))
    Use Spawn Probability=false
    UseLoopCountLimit=false
   )
   &SpawnBurst_Instantaneous(
    Age:NiagaraFloat=$Emitter.Age
    Spawn Count=&MakeIntFromFloat(Float=&FN_FrontEndFloatModifier(Gameplay=0, Front End=2))
    Spawn Time=0.5
    Use Spawn Probability=false
    UseLoopCountLimit=false
   ) {module=SpawnBurst_Instantaneous001}
   &SpawnBurst_Instantaneous(
    Age:NiagaraFloat=$Emitter.Age
    Spawn Count=&MakeIntFromFloat(Float=&FN_FrontEndFloatModifier(Gameplay=0, Front End=2))
    Spawn Time=1
    Use Spawn Probability=false
    UseLoopCountLimit=false
   ) {module=SpawnBurst_Instantaneous002}
   &InterpolateOverTime(Rate Of Change=50, Target Vector Value:Vector3f=$Engine.Owner.Velocity, Variable Type=::NewEnumerator2)
   /SetVariables(Emitter.ExhaustBlend=0.5) {isSetParameters=true, module=SetVariables_#6}
   /SetVariables(Emitter.WindWeight:NiagaraFloat=$System.FN_Contrail_Setup.Dive Value) {isSetParameters=true, module=SetVariables_#7}
  ]
  particleSpawnScript [
   &InitializeParticle(
    Color Mode=::NewEnumerator0
    Lifetime=5
    Mesh Scale Mode=::NewEnumerator3
    Mesh Uniform Scale=&Multiply_Float(
     A=$System.Ship Scale
     B=&SelectFloatFromArray(
      Direct Array Index=&ReturnExecIndex()
      Float Selection Array=DataInterface() {propertyValues {FloatData [0.7,0.7,0.7,0.7,0.7]}}
     )
    )
    Position Offset=<0 0 1000>
    UsePositionOffset=true
    Write Material Random=true
   )
   &ShapeLocation(MaskSpawnGroup=false, Override Seed=false, UseOwnerScale=true)
   &InitialMeshOrientation(Mesh Orientation Mode=::NewEnumerator7, Use Rotation Vector=true)
   &FN_SetSpawnGroup(Number of Spawn Groups=5)
   /SetVariables(Particles.ShipIndex:NiagaraInt32=&MergeHelper_IntSwitch(0=0, 1=1, 2=1, 3=1, 4=1, 5=?, Index=$Emitter.SpawnGroup, Num=::NewEnumerator3)) {isSetParameters=true}
   /SetVariables(Particles.ShipNum:NiagaraInt32=$Emitter.SpawnGroup) {isSetParameters=true, module=SetVariables_#8}
  ]
  particleUpdateScript [
   &ParticleState(Kill Particles When Lifetime Has Elapsed=false, Let Infinitely Lived Particles Die When Emitter Deactivates=true)
   /SetVariables(Particles.Ambient:Vector3f=&Multiply_VectorByFloat(
     Vector=&MakeVector(
      X=&Waveform(
       "Waveform [1]"=::NewEnumerator12
       "[1] Input Position"=&RandomRangeFloat(Minimum=0, Maximum=1000)
       "[1] Frequency"=1
       "[1] Amplitude Min/Max"=<-50 50>
      )
      Y=&Waveform(
       "Waveform [1]"=::NewEnumerator12
       "[1] Input Position"=&RandomRangeFloat(Minimum=1200, Maximum=1800)
       "[1] Frequency"=1
       "[1] Amplitude Min/Max"=<-50 50>
      )
      Z=&Waveform(
       "Waveform [1]"=::NewEnumerator12
       "[1] Input Position"=&RandomRangeFloat(Minimum=2200, Maximum=2800)
       "[1] Frequency"=1
       "[1] Amplitude Min/Max"=<-50 50>
      )
     )
     Float=&Lerp_Float(A=0.02, B=0.1, Alpha=$System.FN_Contrail_Setup.Dive Value)
    )) {isSetParameters=true, module=SetVariables_#9}
   /SetVariables(Particles.CatchupVector:Vector3f=&Multiply_VectorByFloat(
     Vector=&Multiply_VectorByFloat(
      Vector=&NormalizeVector(Vector To Normalize=$System.SkydiveVelocity, Safe Normalize=true, Safe Fallback Threshold=0, Safe Fallback Vector=<0 0 0.001>)
      Float=-75
     )
     Float=&FloatFromCurve(
      FloatCurve=DataInterface() {propertyValues {Curve {keys [
          {interpMode=RCIM_Cubic, tangentMode=RCTM_User, time=0, value=-0.15}
          {interpMode=RCIM_Cubic, tangentMode=RCTM_User, time=0.1, value=0.125}
          {tangentMode=RCTM_User, time=0.6, value=-0.125}
         ]}}}
      CurveIndex=$System.Timeline003.Time
      Scale Curve=0
     )
    )) {isSetParameters=true, module=SetVariables_#10}
   &SkeletalMeshLocation(
    Bone Sampling Mode=::NewEnumerator4
    MaskSpawnGroup=false
    Position Sampling=::NewEnumerator2
    Sampled Position Offset=&Add_Vector(
     A=$Particles.CatchupVector
     B=&RotateVector(
      Vector To Rotate:Vector3f=&Add_Vector(
       A=$Particles.Ambient
       B=&Add_Vector(
        A=&Lerp_Vector(
         A=&SelectVectorFromArray(
          Direct Array Index=&ReturnExecIndex()
          Direct Array Mode=::NewEnumerator1
          Vector Selection Array=DataInterface() {propertyValues {FloatData [<0 100 60>, <-25 20 80>, <25 20 80>, <-50 -60 100>, <50 -60 100>]}}
         )
         B=&SelectVectorFromArray(
          Direct Array Index=&ReturnExecIndex()
          Direct Array Mode=::NewEnumerator1
          Vector Selection Array=DataInterface() {propertyValues {FloatData [<0 100 40>, <-25 20 60>, <25 20 60>, <-50 -60 80>, <50 -60 80>]}}
         )
         Alpha=&FloatFromCurve(
          FloatCurve=DataInterface() {propertyValues {Curve {
             defaultValue=0
             keys [
              {arriveTangent=-0.2472, interpMode=RCIM_Cubic, leaveTangent=-0.2472, tangentMode=RCTM_User, time=-0.2, value=0}
              {arriveTangent=0.254, tangentMode=RCTM_User, time=0.8, value=1}
             ]
            }}}
          CurveIndex=$System.FN_Contrail_Setup.Dive Value
         )
        )
        B=&Lerp_Vector(
         A=<0 0 500>
         B=<0 0 75>
         Alpha=&FloatFromCurve(
          FloatCurve=DataInterface() {propertyValues {Curve {keys [{arriveTangent=1.2212, interpMode=RCIM_Cubic, leaveTangent=1.2212, tangentMode=RCTM_User, time=0, value=0}, <at 3 1 RCIM_Cubic>]}}}
          CurveIndex=$Particles.Age
         )
        )
       )
      )
      Yaw:NiagaraFloat=&Waveform(
       "Waveform [1]"=::NewEnumerator12
       "[1] Input Position"=&RandomRangeFloat(Minimum=200000, Maximum=100000)
       "[1] Frequency"=1.25
       "[1] Amplitude Min/Max"=<7 -7>
       "[1] Scale"=0.2
      )
      Pitch:NiagaraFloat=0
      Roll:NiagaraFloat=0
      Rotation Rate:NiagaraFloat=1
     )
    )
    Skeletal Mesh=$System.SkeletalMesh
    Use Sampled Mesh Attribute as Ribbon ID=false
   )
   &ScaleMeshSize(Scale Factor=&VectorFromFloat(Value=&FloatFromCurve(
      FloatCurve=DataInterface() {propertyValues {Curve {keys [{arriveTangent=1.0186, interpMode=RCIM_Cubic, leaveTangent=1.0186, tangentMode=RCTM_User, time=0, value=0}, <at 0.3 1 RCIM_Cubic>]}}}
      CurveIndex=$Particles.Age
     )))
   &Drag(Drag=2, Use Linear Drag=true, Use Rotational Drag=false)
   &SolveForcesAndVelocity(Clamp Velocity=false, Limit Acceleration=false)
   &UpdateMeshOrientation(
    Bank Rate=900000000
    Banking Angle Min / Max:Vector2f=<0.1 -0.1>
    Orientation Method=::NewEnumerator3
    Particle Velocity:Vector3f=&Multiply_VectorByFloat(
     Vector=&Add_Vector(
      A=&ReturnMeshOrientationAxis(Mesh Orientation Axis=::YAxis, Mesh Orientation Quat=$Particles.SkeletalMeshLocation.SampledRotation)
      B=&Lerp_Vector(A=<0 0 -0.3>, B=<0 0 -2>, Alpha=$System.FN_Contrail_Setup.Dive Value)
     )
     Float=500000
    )
    Rotation Decay Rate In Degrees=900000000
    Up Vector:Vector3f=&Lerp_Vector(
     A=&ReturnMeshOrientationAxis(Mesh Orientation Axis=::YAxis, Mesh Orientation Quat=$Particles.SkeletalMeshLocation.SampledRotation)
     B=&Multiply_VectorByFloat(
      Vector=&ReturnMeshOrientationAxis(Mesh Orientation Axis=::ZAxis, Mesh Orientation Quat=$Particles.SkeletalMeshLocation.SampledRotation)
      Float=1
     )
     Alpha=&FloatFromCurve(
      FloatCurve=DataInterface() {propertyValues {Curve {keys [{arriveTangent=-0.0511, interpMode=RCIM_Cubic, leaveTangent=-0.0511, tangentMode=RCTM_User, time=0, value=0}, <at 1 1>]}}}
      CurveIndex=$System.Timeline003.Time
     )
    )
    Upright Roll Correction Speed=900000000
    Velocity Orientation Convergence Rate=900000000
   ) {module=UpdateMeshOrientation001}
   &InterpolateOverTime(
    Initial Vector Value:NiagaraPosition=$Engine.Owner.Position
    Rate Of Change=50
    Target Vector Value:Vector3f=$Particles.SkeletalMeshLocation.SampledPosition
    Variable Type=::NewEnumerator2
   ) {module=InterpolateOverTime001}
   &InterpolateOverTime(Rate Of Change=10, Target Quat Value:Quat4f=$Particles.MeshOrientation, Variable Type=::NewEnumerator5) {module=InterpolateOverTime002}
   /SetVariables(Particles.Position=$Particles.InterpolateOverTime001.InterpolateOverTime_V3_.Moving Average) {isSetParameters=true}
   /SetVariables(Particles.New Orient:Quat4f=$Particles.InterpolateOverTime002.InterpolateOverTime_Quaternion.Moving Average) {isSetParameters=true}
   /SetVariables(Particles.MeshOrientation=$Particles.New Orient) {isSetParameters=true}
   /SetVariables(Particles.SpeedConeColor:LinearColor=&MakeLinearColorFromVectorAndFloat(
     "Vector (RGB)"=<1 1 1>
     "Float (Alpha)"=&Multiply_Float(
      A=$System.FN_Contrail_Setup.Dive Value
      B=&FloatFromCurve(
       FloatCurve=DataInterface() {propertyValues {Curve {
          defaultValue=1
          keys [
           {interpMode=RCIM_Cubic, tangentMode=RCTM_User, time=0.5, value=1.25}
           {interpMode=RCIM_Cubic, tangentMode=RCTM_User, time=0.6, value=5}
           {tangentMode=RCTM_User, time=1, value=1}
          ]
         }}}
       CurveIndex=$System.Timeline002.Time
       Scale Curve=&Multiply_Float(
        A=0.1
        B=&Waveform(
         Global Input Position=&RandomRangeFloat(Minimum=0, Maximum=1)
         "Waveform [1]"=::NewEnumerator12
         "[1] Frequency"=&Lerp_Float(A=8, B=12, Alpha=$System.FN_Contrail_Setup.Dive Value)
         "[1] Amplitude Min/Max"=<0.75 1.2>
        )
       )
      )
     )
    )) {isSetParameters=true, module=SetVariables_#11}
   /SetVariables(Particles.ExhaustPosition:NiagaraPosition=&TransformVectorByMatrix(
     Vector:Vector3f=<-5 0 0>
     Transformation Matrix:NiagaraMatrix=&MakeTransform(
      Inverse:NiagaraBool=false
      Position:Vector3f=Linked() {linkedVariable {name=Particles.Position, type=NiagaraPosition}}
      Local Offset:Vector3f=<0 0 0>
      Scale:Vector3f=$Particles.Scale
      Rotation Interface:ENiagara_RotationDefinitionApproach=::NewEnumerator1
      Axis:Vector3f=?
      Quaternion:Quat4f=$Particles.MeshOrientation
      Rotation Matrix:NiagaraMatrix=?
      X Vector:Vector3f=?
      Yaw:NiagaraFloat=?
      Angle:NiagaraFloat=?
      Pitch:NiagaraFloat=?
      Y Vector:Vector3f=?
      Roll:NiagaraFloat=?
      Z Vector:Vector3f=?
     )
     Transform Vector as Position:NiagaraBool=true
    )) {isSetParameters=true, module=SetVariables_#12}
   &DynamicMaterialParameters(
    Index 0 Param 1=&SelectFloatFromArray(
     Direct Array Index=&ReturnExecIndex()
     Float Selection Array=DataInterface() {propertyValues {FloatData [0,0.5]}}
    )
    Index 0 Param 2=&RandomRangeFloat(Minimum=0, Maximum=1)
    Index 1 Param 1=&MakeFloatFromVector(
     Vector=&ReturnMeshOrientationAxis(Mesh Orientation Axis=::XAxis, Mesh Orientation Quat=$Particles.MeshOrientation)
     Channel=::NewEnumerator0
    )
    Index 1 Param 2=&MakeFloatFromVector(
     Vector=&ReturnMeshOrientationAxis(Mesh Orientation Axis=::XAxis, Mesh Orientation Quat=$Particles.MeshOrientation)
     Channel=::NewEnumerator1
    )
    Index 1 Param 3=&MakeFloatFromVector(
     Vector=&ReturnMeshOrientationAxis(Mesh Orientation Axis=::XAxis, Mesh Orientation Quat=$Particles.MeshOrientation)
     Channel=::NewEnumerator2
    )
    Index 1 Param 4=&Lerp_Float(
     A=-0.03
     B=0.01
     Alpha=&FloatFromCurve(
      FloatCurve=DataInterface() {propertyValues {Curve {keys [{arriveTangent=-0.0511, interpMode=RCIM_Cubic, leaveTangent=-0.0511, tangentMode=RCTM_User, time=0, value=0}, <at 1 1>]}}}
      CurveIndex=$System.Timeline003.Time
     )
    )
    Module.Param0WriteEnabled=true
    Module.Param0WriteEnabled1=true
    Module.Param1WriteEnabled=true
    Module.Param1WriteEnabled1=true
    Module.Param2WriteEnabled=true
    Module.Param2WriteEnabled1=true
    Module.Param3WriteEnabled=true
    Module.Param3WriteEnabled1=true
    Write Parameter Index 1=true
   )
   &FN_Mesh_Attributes(
    Color=&MultiplyLinearColor(
     Linear Color=<rgba 1 1 1 1>
     Scale Factor=&MakeVector4FromVectorAndFloat(
      "Vector (RGB)"=&VectorFromFloat(Value=&Waveform(
        Global Input Position=&RandomRangeFloat(Minimum=0, Maximum=1)
        "Waveform [1]"=::NewEnumerator12
        "[1] Frequency"=3
        "[1] Amplitude Min/Max"=<0.9 5>
       ))
      "Float (Alpha)"=0.5
     )
    )
    Create Color Attribute=true
    Create Scale Attribute=true
    Scale=&Multiply_Vector(
     A=&MakeVector(X=&Lerp_Float(A=0, B=1, Alpha=$Emitter.ExhaustBlend), Y=1, Z=1)
     B=&Lerp_Vector(
      A=&Lerp_Vector(
       A=<0.1 1.1 0.6>
       B=<0.1 1 1>
       Alpha=&FloatFromCurve(
        FloatCurve=DataInterface() {propertyValues {Curve {keys [{arriveTangent=-0.0511, interpMode=RCIM_Cubic, leaveTangent=-0.0511, tangentMode=RCTM_User, time=0, value=0}, <at 1 1>]}}}
        CurveIndex=$System.Timeline003.Time
       )
      )
      B=&Multiply_Vector(
       A=&Lerp_Vector(
        A=<0.1 1.1 0.6>
        B=<0.1 1 1>
        Alpha=&FloatFromCurve(
         FloatCurve=DataInterface() {propertyValues {Curve {keys [{arriveTangent=-0.0511, interpMode=RCIM_Cubic, leaveTangent=-0.0511, tangentMode=RCTM_User, time=0, value=0}, <at 1 1>]}}}
         CurveIndex=$System.Timeline003.Time
        )
       )
       B=<3 1 1>
      )
      Alpha=$System.FN_Contrail_Setup.Dive Value
     )
    )
    Write to Module Namespace=true
   ) {module=FN_Mesh_Attributes001}
   &FN_Mesh_Attributes(
    Color=&MultiplyLinearColor(
     Linear Color=<rgba 1 1 1 1>
     Scale Factor=&MakeVector4FromVectorAndFloat(
      "Vector (RGB)"=<1 1 1>
      "Float (Alpha)"=&FloatFromCurve(
       FloatCurve=DataInterface() {propertyValues {Curve {
          defaultValue=1
          keys [
           {arriveTangent=-0.2012, interpMode=RCIM_Cubic, leaveTangent=-0.2012, tangentMode=RCTM_User, time=0.5, value=0}
           {arriveTangent=4.0969, interpMode=RCIM_Cubic, leaveTangent=4.0969, tangentMode=RCTM_User, time=1, value=1}
          ]
         }}}
       CurveIndex=$Emitter.WindWeight
       Scale Curve=0.02
      )
     )
    )
    Create Color Attribute=true
    Create Scale Attribute=true
    Scale=&Multiply_Vector(A=$Particles.Scale, B=&Lerp_Vector(A=<1 1 1>, B=<2 1 1>, Alpha=$Emitter.WindWeight))
    Write to Module Namespace=true
   ) {module=FN_Mesh_Attributes002}
  ]
  rendererClasses [&NiagaraMeshRendererProperties]
  renderers [
   {rendererClass=&NiagaraMeshRendererProperties, rendererIndex=0}
   {rendererClass=&NiagaraMeshRendererProperties, rendererIndex=1}
   {rendererClass=&NiagaraMeshRendererProperties, rendererIndex=2}
   {rendererClass=&NiagaraMeshRendererProperties, rendererIndex=3}
  ]
 }
 Ship_Trails {
  emitterUpdateScript [
   &EmitterState(Life Cycle Mode=::NewEnumerator1, Loop Delay=0.01, Loop Duration=1, UseLoopDelay=false)
   &SpawnParticlesFromOtherEmitter(
    Attribute Reader=DataInterface() {propertyValues {EmitterBinding {bindingMode=Other, emitterName=Ship_FE}}}
    Calculate Spawn Rate Per Particle=false
    Spawn Rate=40
    Use Maximum Age=false
    Use Minimum Age=false
   )
  ]
  particleSpawnScript [
   &FN_SetSpawnGroup()
   &InitializeParticle(
    Color Mode=::NewEnumerator0
    Lifetime=&Lerp_Float(A=1, B=1, Alpha=$System.Dive)
    UsePositionOffset=false
    Write Material Random=true
   )
   &SampleParticlesFromOtherEmitter(
    Applied Velocity Scale=0
    Mesh Orientation Sampling:ENiagara_DirectReadSamplingMode=::NewEnumerator1
    Particle ID Sampling=::NewEnumerator1
    Velocity Application:ENiagara_DirectReadApplicationMode=::NewEnumerator0
    Velocity Sampling=::NewEnumerator0
   )
  ]
  particleUpdateScript [
   &ParticleState()
   &ScaleColor(
    Scale Alpha=&FloatFromCurve(
     FloatCurve=DataInterface() {propertyValues {Curve {
        defaultValue=1
        keys [
         {interpMode=RCIM_Cubic, tangentMode=RCTM_User, time=0, value=0}
         {interpMode=RCIM_Cubic, tangentMode=RCTM_User, time=0.5, value=1}
         {tangentMode=RCTM_User, time=1, value=0}
        ]
       }}}
     Scale Curve=&Lerp_Float(A=0.1, B=0.5, Alpha=$System.FN_Contrail_Setup.Dive Value)
    )
    Scale RGB=&MakeVectorFromLinearColorRGB(Color=<rgba 2 0.16 0.4053 1>)
    ScaleA=true
    ScaleRGB=true
   )
   &SolveForcesAndVelocity(Clamp Velocity=false, Limit Acceleration=false)
   &self:SampleShip_04(New Particle Attribute Reader=DataInterface() {propertyValues {EmitterBinding {bindingMode=Other, emitterName=Ship_FE}}}) {module=SampleShip_04}
   &FN_Sprite_Attributes(
    Camera Offset=5
    Create Camera Offset Attribute=true
    Create Sprite Rotation Attribute=true
    Create SpriteSize Attribute=true
    SpriteRotation:NiagaraFloat=&RandomRangeFloat(Minimum=0, Maximum=1)
    SpriteSize=&Multiply_Vector2D_ByFloat(
     Float=&Waveform(
      Global Input Position=&RandomRangeFloat(Minimum=0, Maximum=1)
      "Waveform [1]"=::NewEnumerator12
      "[1] Frequency"=&Lerp_Float(A=8, B=12, Alpha=$System.FN_Contrail_Setup.Dive Value)
      "[1] Amplitude Min/Max"=<0.75 1.2>
     )
     Vector2D=&Multiply_Vector2D_ByFloat(
      Float=&FloatFromCurve(
       FloatCurve=DataInterface() {propertyValues {Curve {
          defaultValue=1
          keys [
           {interpMode=RCIM_Cubic, tangentMode=RCTM_User, time=0.5, value=1}
           {interpMode=RCIM_Cubic, tangentMode=RCTM_User, time=0.6, value=2.5}
           {tangentMode=RCTM_User, time=1, value=1}
          ]
         }}}
       CurveIndex=$System.Timeline002.Time
      )
      Vector2D=<15 15>
     )
    )
    Write to Module Namespace=true
   )
   /SetVariables(Particles.Position=$Output.SampleShip_04.Ship Pos) {isSetParameters=true, module=SetVariables_#13}
   /SetVariables(Particles.OffsetPosition:NiagaraPosition=&AddVectorToPosition(
     Position=$Particles.Position
     Vector=&Lerp_Vector(
      A=&RotateVectorByQuaternion(
       Vector To Rotate=&SelectVectorFromArray(
        Direct Array Index=&SpawnGroup()
        Vector Selection Array=DataInterface() {propertyValues {FloatData [<-11 6 2>, <-11 6 -2>, <-11 -6 2>, <-11 -6 -2>]}}
       )
       Quaternion=$Output.SampleShip_04.Ship Rot
      )
      B=&RotateVectorByQuaternion(
       Vector To Rotate=&SelectVectorFromArray(
        Direct Array Index=&SpawnGroup()
        Vector Selection Array=DataInterface() {propertyValues {FloatData [<-11 5.2 2.8>, <-11 5.2 -2.8>, <-11 -5.2 2.8>, <-11 -5.2 -2.8>]}}
       )
       Quaternion=$Output.SampleShip_04.Ship Rot
      )
      Alpha=&FloatFromCurve(
       FloatCurve=DataInterface() {propertyValues {Curve {keys [{arriveTangent=-0.0511, interpMode=RCIM_Cubic, leaveTangent=-0.0511, tangentMode=RCTM_User, time=0, value=0}, <at 1 1>]}}}
       CurveIndex=$System.Timeline003.Time
      )
     )
    )) {isSetParameters=true, module=SetVariables_#14}
  ]
  rendererClasses [&NiagaraSpriteRendererProperties]
  renderers [{rendererClass=&NiagaraSpriteRendererProperties}]
 }
 Ship_Trails001 {
  emitterUpdateScript [
   &EmitterState(Life Cycle Mode=::NewEnumerator1, Loop Delay=0.01, Loop Duration=1, UseLoopDelay=false)
   &SpawnParticlesFromOtherEmitter(
    Attribute Reader=DataInterface() {propertyValues {EmitterBinding {bindingMode=Other, emitterName=ship}}}
    Calculate Spawn Rate Per Particle=false
    Spawn Rate=40
    Use Maximum Age=false
    Use Minimum Age=false
   )
  ]
  particleSpawnScript [
   &FN_SetSpawnGroup()
   &InitializeParticle(
    Color Mode=::NewEnumerator0
    Lifetime=&Lerp_Float(A=1, B=1, Alpha=$System.Dive)
    UsePositionOffset=false
    Write Material Random=true
   )
   &SampleParticlesFromOtherEmitter(
    Applied Velocity Scale=0
    Mesh Orientation Sampling:ENiagara_DirectReadSamplingMode=::NewEnumerator1
    Particle ID Sampling=::NewEnumerator1
    Velocity Application:ENiagara_DirectReadApplicationMode=::NewEnumerator0
    Velocity Sampling=::NewEnumerator0
   )
  ]
  particleUpdateScript [
   &ParticleState()
   &ScaleColor(
    Scale Alpha=&FloatFromCurve(
     FloatCurve=DataInterface() {propertyValues {Curve {
        defaultValue=1
        keys [
         {interpMode=RCIM_Cubic, tangentMode=RCTM_User, time=0, value=0}
         {interpMode=RCIM_Cubic, tangentMode=RCTM_User, time=0.5, value=1}
         {tangentMode=RCTM_User, time=1, value=0}
        ]
       }}}
     Scale Curve=&Lerp_Float(A=0.2, B=0.6, Alpha=$System.FN_Contrail_Setup.Dive Value)
    )
    Scale RGB=&MakeVectorFromLinearColorRGB(Color=<rgba 2 0.16 0.4053 1>)
    ScaleA=true
    ScaleRGB=true
   )
   &SolveForcesAndVelocity(Clamp Velocity=false, Limit Acceleration=false)
   &self:SampleShip_04(New Particle Attribute Reader=DataInterface() {propertyValues {EmitterBinding {bindingMode=Other, emitterName=ship}}}) {module=SampleShip_04}
   &FN_Sprite_Attributes(
    Camera Offset=5
    Create Camera Offset Attribute=true
    Create Sprite Rotation Attribute=true
    Create SpriteSize Attribute=true
    SpriteRotation:NiagaraFloat=&RandomRangeFloat(Minimum=0, Maximum=1)
    SpriteSize=&Multiply_Vector2D_ByFloat(
     Float=&Waveform(
      Global Input Position=&RandomRangeFloat(Minimum=0, Maximum=1)
      "Waveform [1]"=::NewEnumerator12
      "[1] Frequency"=&Lerp_Float(A=8, B=12, Alpha=$System.FN_Contrail_Setup.Dive Value)
      "[1] Amplitude Min/Max"=<0.75 1.2>
     )
     Vector2D=&Multiply_Vector2D_ByFloat(
      Float=&FloatFromCurve(
       FloatCurve=DataInterface() {propertyValues {Curve {
          defaultValue=1
          keys [
           {interpMode=RCIM_Cubic, tangentMode=RCTM_User, time=0.5, value=1}
           {interpMode=RCIM_Cubic, tangentMode=RCTM_User, time=0.6, value=2.5}
           {tangentMode=RCTM_User, time=1, value=1}
          ]
         }}}
       CurveIndex=$System.Timeline002.Time
      )
      Vector2D=<15 15>
     )
    )
    Write to Module Namespace=true
   )
   /SetVariables(Particles.Position=$Output.SampleShip_04.Ship Pos) {isSetParameters=true, module=SetVariables_#13}
   /SetVariables(Particles.OffsetPosition:NiagaraPosition=&AddVectorToPosition(
     Position=$Particles.Position
     Vector=&Lerp_Vector(
      A=&RotateVectorByQuaternion(
       Vector To Rotate=&SelectVectorFromArray(
        Direct Array Index=&SpawnGroup()
        Vector Selection Array=DataInterface() {propertyValues {FloatData [<-11 6 2>, <-11 6 -2>, <-11 -6 2>, <-11 -6 -2>]}}
       )
       Quaternion=$Output.SampleShip_04.Ship Rot
      )
      B=&RotateVectorByQuaternion(
       Vector To Rotate=&SelectVectorFromArray(
        Direct Array Index=&SpawnGroup()
        Vector Selection Array=DataInterface() {propertyValues {FloatData [<-11 5.2 2.8>, <-11 5.2 -2.8>, <-11 -5.2 2.8>, <-11 -5.2 -2.8>]}}
       )
       Quaternion=$Output.SampleShip_04.Ship Rot
      )
      Alpha=&FloatFromCurve(
       FloatCurve=DataInterface() {propertyValues {Curve {keys [{arriveTangent=-0.0511, interpMode=RCIM_Cubic, leaveTangent=-0.0511, tangentMode=RCTM_User, time=0, value=0}, <at 1 1>]}}}
       CurveIndex=$System.Timeline003.Time
      )
     )
    )) {isSetParameters=true, module=SetVariables_#14}
  ]
  rendererClasses [&NiagaraSpriteRendererProperties]
  renderers [{rendererClass=&NiagaraSpriteRendererProperties}]
 }
}
summary {
 systemName=NS_Contrail_MarkerDeer
 emitters [
  {emitterName=Ship, rendererClasses [&NiagaraMeshRendererProperties]}
  {emitterName=FN_SpeedLines, simTarget=GPUComputeSim, rendererClasses [&NiagaraSpriteRendererProperties]}
  {emitterName=Ship_Trails001, rendererClasses [&NiagaraSpriteRendererProperties]}
  {emitterName=Ship_Trails, rendererClasses [&NiagaraSpriteRendererProperties]}
  {emitterName=Ship_FE, rendererClasses [&NiagaraMeshRendererProperties]}
 ]
 userVariables [
  User.bIsFrontEnd {type=NiagaraBool}
  User.bIsFrontEndPreview {type=NiagaraBool}
  User.ParaGlideLeanAlpha {type=NiagaraFloat, defaultValue=0, description=""}
  User.PlayerSpeed {type=NiagaraFloat, defaultValue=0, description=""}
  User.RibbonIDOffset {type=NiagaraInt32}
 ]
}
